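# Anim Blueprint that adds physics sway to eight ribbon bones on the ForestJeanWink drumstick backpack cosmetic.
# Two serial AnimDynamics passes cover the same bones with angular limits and damping; the second has higher acceleration clamps and runs only outside the front end.
# The EventGraph sets a GravityOverride variable from skydiving and parachute state, but the nodes' gravity pins show fixed values, not that variable.
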AnimBlueprint /SparksCosmetics/Instrument/DrumStick/ForestJeanWink/Meshes/ForestJeanWink_DrumStick_Backpack_AnimBP
target_skeleton=&/ForestJeanWink_DrumStick_Backpack_Skeleton
graphs [
 AnimGraph {nodes [
   n1 = Anim Dynamics - Bone: dyn_ribbon_01!(
    ComponentPose=Local To Component(LocalPose=Input Pose())
    bAlphaBoolEnabled:bool=True
    GravityOverride:Vector="0,0,-980"
    ExternalForce="0,0,0"
   ) -> (Pose) {
    alphaInputType=Bool
    angularDampingOverride=0.99
    bOverrideAngularDamping=true
    bUseGravityOverride=true
    alphaBoolBlend {blendOption=Linear}
    boundBone {boneName=dyn_ribbon_01}
    componentAppliedLinearAccClamp=<200 200 200>
    componentLinearAccScale=<0.5 0.5 0.5>
    componentLinearVelScale=<0.5 0.5 0.5>
    gravityOverride {z=-980}
    physicsBodyDefinitions [{
      boundBone {boneName=dyn_ribbon_01}
      boxExtents=<10 10 10>
      constraintSetup {twistAxis=AxisZ, angularLimitsMax=<40 40 40>, angularLimitsMin=<-40 -40 -30>, angularTarget=<0 0 0>}
      localJointOffset=<1 0 0>
     }]
   }
   n2 = Anim Dynamics - Bone: dyn_ribbon_02!(ComponentPose=@n1, bAlphaBoolEnabled:bool=True, GravityOverride:Vector="0,0,-980", ExternalForce="0,0,0") -> (Pose) {
    alphaInputType=Bool
    angularDampingOverride=0.99
    bOverrideAngularDamping=true
    bUseGravityOverride=true
    linearDampingOverride=0.9
    alphaBoolBlend {blendOption=Linear}
    boundBone {boneName=dyn_ribbon_02}
    componentAppliedLinearAccClamp=<200 200 200>
    componentLinearAccScale=<0.6 0.6 0.6>
    componentLinearVelScale=<0.5 0.5 0.5>
    physicsBodyDefinitions [{
      boundBone {boneName=dyn_ribbon_02}
      boxExtents=<10 10 10>
      constraintSetup {angularLimitsMax=<40 30 40>, angularLimitsMin=<-30 -30 -30>, angularTarget=<0 0 0>}
      localJointOffset=<1 0 0>
     }]
   }
   n3 = Anim Dynamics - Bone: dyn_ribbon_03!(ComponentPose=@n2, bAlphaBoolEnabled:bool=True, GravityOverride:Vector="0,0,-980", ExternalForce="0,0,0") -> (Pose) {
    alphaInputType=Bool
    angularDampingOverride=0.99
    bOverrideAngularDamping=true
    bUseGravityOverride=true
    linearDampingOverride=0.9
    alphaBoolBlend {blendOption=Linear}
    boundBone {boneName=dyn_ribbon_03}
    componentAppliedLinearAccClamp=<200 200 200>
    componentLinearAccScale=<0.6 0.6 0.6>
    componentLinearVelScale=<0.5 0.5 0.5>
    physicsBodyDefinitions [{
      boundBone {boneName=dyn_ribbon_03}
      boxExtents=<10 10 10>
      constraintSetup {angularLimitsMax=<40 30 40>, angularLimitsMin=<-30 -30 -30>, angularTarget=<0 0 0>}
      localJointOffset=<1 0 0>
     }]
   }
   n4 = Anim Dynamics - Bone: dyn_ribbon_04!(ComponentPose=@n3, bAlphaBoolEnabled:bool=True, GravityOverride:Vector="0,0,-980", ExternalForce="0,0,0") -> (Pose) {
    alphaInputType=Bool
    angularDampingOverride=0.99
    bOverrideAngularDamping=true
    bUseGravityOverride=true
    linearDampingOverride=0.9
    alphaBoolBlend {blendOption=Linear}
    boundBone {boneName=dyn_ribbon_04}
    componentAppliedLinearAccClamp=<200 200 200>
    componentLinearAccScale=<0.6 0.6 0.6>
    componentLinearVelScale=<0.5 0.5 0.5>
    physicsBodyDefinitions [{
      boundBone {boneName=dyn_ribbon_04}
      boxExtents=<10 10 10>
      constraintSetup {angularLimitsMax=<40 60 60>, angularLimitsMin=<-30 -50 -50>, angularTarget=<0 0 0>}
      localJointOffset=<1 0 0>
     }]
   }
   n5 = Anim Dynamics - Bone: dyn_ribbon_05!(ComponentPose=@n4, bAlphaBoolEnabled:bool=True, GravityOverride:Vector="0,0,-980", ExternalForce="0,0,0") -> (Pose) {
    alphaInputType=Bool
    angularDampingOverride=0.99
    bOverrideAngularDamping=true
    bUseGravityOverride=true
    alphaBoolBlend {blendOption=Linear}
    boundBone {boneName=dyn_ribbon_05}
    componentAppliedLinearAccClamp=<200 200 200>
    componentLinearAccScale=<0.5 0.5 0.5>
    componentLinearVelScale=<0.5 0.5 0.5>
    gravityOverride {z=-980}
    physicsBodyDefinitions [{
      boundBone {boneName=dyn_ribbon_05}
      boxExtents=<10 10 10>
      constraintSetup {twistAxis=AxisZ, angularLimitsMax=<40 40 40>, angularLimitsMin=<-40 -40 -30>, angularTarget=<0 0 0>}
      localJointOffset=<1 0 0>
     }]
   }
   n6 = Anim Dynamics - Bone: dyn_ribbon_06!(ComponentPose=@n5, bAlphaBoolEnabled:bool=True, GravityOverride:Vector="0,0,-980", ExternalForce="0,0,0") -> (Pose) {
    alphaInputType=Bool
    angularDampingOverride=0.99
    bOverrideAngularDamping=true
    bUseGravityOverride=true
    linearDampingOverride=0.9
    alphaBoolBlend {blendOption=Linear}
    boundBone {boneName=dyn_ribbon_06}
    componentAppliedLinearAccClamp=<200 200 200>
    componentLinearAccScale=<0.6 0.6 0.6>
    componentLinearVelScale=<0.5 0.5 0.5>
    physicsBodyDefinitions [{
      boundBone {boneName=dyn_ribbon_06}
      boxExtents=<10 10 10>
      constraintSetup {angularLimitsMax=<40 30 40>, angularLimitsMin=<-30 -30 -30>, angularTarget=<0 0 0>}
      localJointOffset=<1 0 0>
     }]
   }
   n7 = Anim Dynamics - Bone: dyn_ribbon_07!(ComponentPose=@n6, bAlphaBoolEnabled:bool=True, GravityOverride:Vector="0,0,-980", ExternalForce="0,0,0") -> (Pose) {
    alphaInputType=Bool
    angularDampingOverride=0.99
    bOverrideAngularDamping=true
    bUseGravityOverride=true
    linearDampingOverride=0.9
    alphaBoolBlend {blendOption=Linear}
    boundBone {boneName=dyn_ribbon_07}
    componentAppliedLinearAccClamp=<200 200 200>
    componentLinearAccScale=<0.6 0.6 0.6>
    componentLinearVelScale=<0.5 0.5 0.5>
    physicsBodyDefinitions [{
      boundBone {boneName=dyn_ribbon_07}
      boxExtents=<10 10 10>
      constraintSetup {angularLimitsMax=<40 30 40>, angularLimitsMin=<-30 -30 -30>, angularTarget=<0 0 0>}
      localJointOffset=<1 0 0>
     }]
   }
   n8 = Anim Dynamics - Bone: dyn_ribbon_08!(ComponentPose=@n7, bAlphaBoolEnabled:bool=True, GravityOverride:Vector="0,0,-980", ExternalForce="0,0,0") -> (Pose) {
    alphaInputType=Bool
    angularDampingOverride=0.99
    bOverrideAngularDamping=true
    bUseGravityOverride=true
    linearDampingOverride=0.9
    alphaBoolBlend {blendOption=Linear}
    boundBone {boneName=dyn_ribbon_08}
    componentAppliedLinearAccClamp=<200 200 200>
    componentLinearAccScale=<0.6 0.6 0.6>
    componentLinearVelScale=<0.5 0.5 0.5>
    physicsBodyDefinitions [{
      boundBone {boneName=dyn_ribbon_08}
      boxExtents=<10 10 10>
      constraintSetup {angularLimitsMax=<40 60 60>, angularLimitsMin=<-30 -50 -50>, angularTarget=<0 0 0>}
      localJointOffset=<1 0 0>
     }]
   }
   n9 = NOT Boolean(A=Get bIsInFrontEnd())
   n10 = Anim Dynamics - Bone: dyn_ribbon_01(ComponentPose=@n8, bAlphaBoolEnabled=@n9, GravityOverride="0,0,-980", ExternalForce="0,0,0") {
    alphaInputType=Bool
    angularDampingOverride=0.99
    bOverrideAngularDamping=true
    bUseGravityOverride=true
    alphaBoolBlend {blendOption=Linear}
    boundBone {boneName=dyn_ribbon_01}
    componentAppliedLinearAccClamp=<500 500 500>
    componentLinearAccScale=<0.5 0.5 0.5>
    componentLinearVelScale=<0.8 0.8 0.8>
    gravityOverride {z=-980}
    physicsBodyDefinitions [{
      boundBone {boneName=dyn_ribbon_01}
      boxExtents=<10 10 10>
      constraintSetup {twistAxis=AxisZ, angularLimitsMax=<40 40 40>, angularLimitsMin=<-40 -40 -30>, angularTarget=<0 0 0>}
      localJointOffset=<1 0 0>
     }]
   }
   n11 = Anim Dynamics - Bone: dyn_ribbon_02(ComponentPose=@n10, bAlphaBoolEnabled=@n9, GravityOverride="0,0,-980", ExternalForce="0,0,0") {
    alphaInputType=Bool
    angularDampingOverride=0.99
    bOverrideAngularDamping=true
    bUseGravityOverride=true
    linearDampingOverride=0.9
    alphaBoolBlend {blendOption=Linear}
    boundBone {boneName=dyn_ribbon_02}
    componentAppliedLinearAccClamp=<500 500 500>
    componentLinearAccScale=<0.7 0.7 0.7>
    componentLinearVelScale=<0.4 0.4 0.4>
    physicsBodyDefinitions [{
      boundBone {boneName=dyn_ribbon_02}
      boxExtents=<10 10 10>
      constraintSetup {angularLimitsMax=<40 30 40>, angularLimitsMin=<-30 -30 -30>, angularTarget=<0 0 0>}
      localJointOffset=<1 0 0>
     }]
   }
   n12 = Anim Dynamics - Bone: dyn_ribbon_03(ComponentPose=@n11, bAlphaBoolEnabled=@n9, GravityOverride="0,0,-980", ExternalForce="0,0,0") {
    alphaInputType=Bool
    angularDampingOverride=0.99
    bOverrideAngularDamping=true
    bUseGravityOverride=true
    linearDampingOverride=0.9
    alphaBoolBlend {blendOption=Linear}
    boundBone {boneName=dyn_ribbon_03}
    componentAppliedLinearAccClamp=<500 500 500>
    componentLinearAccScale=<0.7 0.7 0.7>
    componentLinearVelScale=<0.4 0.4 0.4>
    physicsBodyDefinitions [{
      boundBone {boneName=dyn_ribbon_03}
      boxExtents=<10 10 10>
      constraintSetup {angularLimitsMax=<40 30 40>, angularLimitsMin=<-30 -30 -30>, angularTarget=<0 0 0>}
      localJointOffset=<1 0 0>
     }]
   }
   n13 = Anim Dynamics - Bone: dyn_ribbon_04(ComponentPose=@n12, bAlphaBoolEnabled=@n9, GravityOverride="0,0,-980", ExternalForce="0,0,0") {
    alphaInputType=Bool
    angularDampingOverride=0.99
    bOverrideAngularDamping=true
    bUseGravityOverride=true
    linearDampingOverride=0.9
    alphaBoolBlend {blendOption=Linear}
    boundBone {boneName=dyn_ribbon_04}
    componentAppliedLinearAccClamp=<500 500 500>
    componentLinearAccScale=<0.7 0.7 0.7>
    componentLinearVelScale=<0.4 0.4 0.4>
    physicsBodyDefinitions [{
      boundBone {boneName=dyn_ribbon_04}
      boxExtents=<10 10 10>
      constraintSetup {angularLimitsMax=<40 60 60>, angularLimitsMin=<-30 -50 -50>, angularTarget=<0 0 0>}
      localJointOffset=<1 0 0>
     }]
   }
   n14 = Anim Dynamics - Bone: dyn_ribbon_05(ComponentPose=@n13, bAlphaBoolEnabled=@n9, GravityOverride="0,0,-980", ExternalForce="0,0,0") {
    alphaInputType=Bool
    angularDampingOverride=0.99
    bOverrideAngularDamping=true
    bUseGravityOverride=true
    alphaBoolBlend {blendOption=Linear}
    boundBone {boneName=dyn_ribbon_05}
    componentAppliedLinearAccClamp=<500 500 500>
    componentLinearAccScale=<0.5 0.5 0.5>
    componentLinearVelScale=<0.8 0.8 0.8>
    gravityOverride {z=-980}
    physicsBodyDefinitions [{
      boundBone {boneName=dyn_ribbon_05}
      boxExtents=<10 10 10>
      constraintSetup {twistAxis=AxisZ, angularLimitsMax=<40 40 40>, angularLimitsMin=<-40 -40 -30>, angularTarget=<0 0 0>}
      localJointOffset=<1 0 0>
     }]
   }
   n15 = Anim Dynamics - Bone: dyn_ribbon_06(ComponentPose=@n14, bAlphaBoolEnabled=@n9, GravityOverride="0,0,-980", ExternalForce="0,0,0") {
    alphaInputType=Bool
    angularDampingOverride=0.99
    bOverrideAngularDamping=true
    bUseGravityOverride=true
    linearDampingOverride=0.9
    alphaBoolBlend {blendOption=Linear}
    boundBone {boneName=dyn_ribbon_06}
    componentAppliedLinearAccClamp=<500 500 500>
    componentLinearAccScale=<0.7 0.7 0.7>
    componentLinearVelScale=<0.4 0.4 0.4>
    physicsBodyDefinitions [{
      boundBone {boneName=dyn_ribbon_06}
      boxExtents=<10 10 10>
      constraintSetup {angularLimitsMax=<40 30 40>, angularLimitsMin=<-30 -30 -30>, angularTarget=<0 0 0>}
      localJointOffset=<1 0 0>
     }]
   }
   n16 = Anim Dynamics - Bone: dyn_ribbon_07(ComponentPose=@n15, bAlphaBoolEnabled=@n9, GravityOverride="0,0,-980", ExternalForce="0,0,0") {
    alphaInputType=Bool
    angularDampingOverride=0.99
    bOverrideAngularDamping=true
    bUseGravityOverride=true
    linearDampingOverride=0.9
    alphaBoolBlend {blendOption=Linear}
    boundBone {boneName=dyn_ribbon_07}
    componentAppliedLinearAccClamp=<500 500 500>
    componentLinearAccScale=<0.7 0.7 0.7>
    componentLinearVelScale=<0.4 0.4 0.4>
    physicsBodyDefinitions [{
      boundBone {boneName=dyn_ribbon_07}
      boxExtents=<10 10 10>
      constraintSetup {angularLimitsMax=<40 30 40>, angularLimitsMin=<-30 -30 -30>, angularTarget=<0 0 0>}
      localJointOffset=<1 0 0>
     }]
   }
   n17 = Anim Dynamics - Bone: dyn_ribbon_08(ComponentPose=@n16, bAlphaBoolEnabled=@n9, GravityOverride="0,0,-980", ExternalForce="0,0,0") {
    alphaInputType=Bool
    angularDampingOverride=0.99
    bOverrideAngularDamping=true
    bUseGravityOverride=true
    linearDampingOverride=0.9
    alphaBoolBlend {blendOption=Linear}
    boundBone {boneName=dyn_ribbon_08}
    componentAppliedLinearAccClamp=<500 500 500>
    componentLinearAccScale=<0.7 0.7 0.7>
    componentLinearVelScale=<0.4 0.4 0.4>
    physicsBodyDefinitions [{
      boundBone {boneName=dyn_ribbon_08}
      boxExtents=<10 10 10>
      constraintSetup {angularLimitsMax=<40 60 60>, angularLimitsMin=<-30 -50 -50>, angularTarget=<0 0 0>}
      localJointOffset=<1 0 0>
     }]
   }
   Output Pose(Result=Component To Local(ComponentPose=@n17))
  ]}
 EventGraph {nodes [
   Event BlueprintUpdateAnimation()
   > n1 = Branch(Condition=Get bIsSkyDiving())
   > n2 = Branch(Condition=Get bIsParachuteOpened())
   > Set GravityOverride(GravityOverride="0,0,-100")
   Set GravityOverride(execute=@n2.else, GravityOverride="0,0,400")
   Set GravityOverride(execute=@n1.else, GravityOverride="0,0,-980")
   TryGetPawnOwner()
  ]}
]
variables [GravityOverride {type=Vector}]
